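# Material function that adds real-time water interaction normals to a water surface's ambient normal.
# The interaction normal, by default derived from the sampled real-time water data height map, is blended with the ambient normal through BlendAngleCorrectedNormals.
# A "Use Player Interaction Normals" static switch and the render-to-texture feature gate fall back to the plain ambient normal.
# A second blend against the vertex normal is left unconnected.
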MaterialFunction /Figment_S07/LM_7_00/WaterAndWind/MaterialFunctions/Fortnite_WaterInteractions_NormalModification
graph {
 properties {libraryCategoriesText [Fortnite Wind and Water]}
 nodes [
  n1 = FunctionInput(Preview=MaterialFunctionCall!(Normal intensity, "World space offset to calculate normal ") {materialFunction=&/Fortnite_SampleRealTimeWaterData}.Normal from height map) {bUsePreviewValueAsDefault=true, inputName=Water Interaction Normal}
  n2 = FunctionInput() {bUsePreviewValueAsDefault=true, inputName=Ambient Normal, previewValue=<0 0 1 1>}
  n3 = MaterialFunctionCall(BaseNormal=@n1, AdditionalNormal=@n2) {materialFunction=&BlendAngleCorrectedNormals}
  n4 = FunctionInput(Preview=StaticBoolParameter() {defaultValue=true, group=Player interaction, parameterName=Use Player Interaction Normals}) {bUsePreviewValueAsDefault=true, inputName=Use Player Interaction Normals, inputType=FunctionInput_StaticBool}
  FunctionOutput(None=MaterialFunctionCall!(
    With RTT Features=StaticSwitch(True=@n3.Result, False=@n2, Value=@n4)
    WithOut RTT Features=@n2
    With RTT Features V2
    WithOut RTT Features V2
    With RTT Features Att
    WithOut RTT Features Att
   ) {materialFunction=/Figment_S07/LM_7_00/WaterAndWind/EnableOrDisableWind/Fortnite_RenderToTextureEnabled}.V3)
  MaterialFunctionCall(BaseNormal=VertexNormalWS(), AdditionalNormal=@n1) {materialFunction=&BlendAngleCorrectedNormals}
 ]
}
